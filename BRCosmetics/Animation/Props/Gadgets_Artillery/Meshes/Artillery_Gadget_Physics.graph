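# Physics asset for the Artillery gadget prop mesh in the BRCosmetics animation props.
# It defines a single body on the main bone, shaped as a capsule with near-zero length, so it acts as a sphere slightly raised above the bone origin.
# It has no constraints, so the prop simulates and collides as one rigid volume rather than an articulated ragdoll.
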PhysicsAsset /BRCosmetics/Animation/Props/Gadgets_Artillery/Meshes/Artillery_Gadget_Physics
bodies {main [{shapeType=Capsule, length=0.0099, radius=44.4825, shapeName=main_capsule, center=<0 0 7.2334>, rotation=<rot 0 -90 0>}]}
constraints []
